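# Physics asset for the hoodie mesh of the medium male IceCream player body, providing a minimal rig for the hood bone.
# It has one capsule body on dyn_hood, offset and rotated from the bone, and another on spine_05.
# A single constraint links the hood to the spine with both swings and twist set to Free, so its stored angle limits do not restrict rotation.
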PhysicsAsset /Game/Characters/Player/Male/Medium/Bodies/M_MED_IceCream/Meshes/M_MED_IceCream_Hoodie_Physics
bodies {
 dyn_hood [{shapeType=Capsule, length=8.1917, radius=7.2149, center=<3.7732 -10.8558 0.0003>, rotation=<rot 0.0009 17.9922 89.9989>}]
 spine_05 [{shapeType=Capsule, length=18.2142, radius=5.1799}]
}
constraints [{
  bone1Name=dyn_hood
  bone2Name=spine_05
  swing1LimitDegrees=30
  swing1Motion=Free
  swing2LimitDegrees=0
  swing2Motion=Free
  twistLimitDegrees=60
  twistMotion=Free
 }]
