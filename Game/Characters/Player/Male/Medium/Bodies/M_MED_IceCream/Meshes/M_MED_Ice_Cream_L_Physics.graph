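# Physics asset for the left-side jacket and drawstring of the medium male player's Ice Cream body, covering the dyn_jacket bones.
# Capsules on spine_04 and spine_05 form the torso, while boxes and a capsule shape the high and low jacket bodies.
# Both jacket bodies constrain to spine_04 with free swing and twist.
# A four-capsule string chain hangs from the high jacket body, each link carrying tight swing and twist limits.
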PhysicsAsset /Game/Characters/Player/Male/Medium/Bodies/M_MED_IceCream/Meshes/M_MED_Ice_Cream_L_Physics
bodies {
 dyn_jacketHigh_l [{shapeType=Box, extentX=12.7519, extentY=3.3653, extentZ=19.9892, center=<5.7469 4.6324 1.8778>, rotation=<rot 2.5854 -93.475 85.1748>}]
 dyn_jacketLow_l [
  {shapeType=Capsule, length=14.4706, radius=9.0069, center=<-8.335 4.8954 6.5548>, rotation=<rot -32.9591 0 -11.1519>}
  {shapeType=Box, extentX=4.0816, extentY=16.4005, extentZ=30.1864, center=<-7.7934 -1.4004 3.4928>, rotation=<rot -11.301 -90.2654 11.108>}
 ]
 dyn_jacketString_01_l [{shapeType=Capsule, length=5.7029, radius=0.6104, center=<-3.0105 -0.0025 0.0362>, rotation=<rot 0 89.9496 -89.3792>}]
 dyn_jacketString_02_l [{shapeType=Capsule, length=6.5044, radius=0.7807, center=<-3.3194 -0.0023 0.0329>, rotation=<rot 0 89.9496 -89.3792>}]
 dyn_jacketString_03_l [{shapeType=Capsule, length=5.5026, radius=0.8106, center=<-3.0105 -0.0025 0.0362>, rotation=<rot 0 89.9496 -89.3792>}]
 dyn_jacketString_04_l [{shapeType=Capsule, length=2.0693, radius=1.0499, center=<-1.3041 -0.004 0.0547>, rotation=<rot 0 89.9496 -89.3792>}]
 spine_04 [{shapeType=Capsule, length=17.267, radius=16.0369, center=<-0.6 2.7721 0.1141>, rotation=<rot 90 0 -7.6681>}]
 spine_05 [{shapeType=Capsule, length=16.0101, radius=14.9826, center=<-0.8402 -6.6554 -0.6411>, rotation=<rot 1.2287 -132.6573 -90>}]
}
constraints [
 {bone1Name=spine_05, bone2Name=spine_04, swing1LimitDegrees=45, swing2LimitDegrees=45, twistLimitDegrees=45}
 {
  bone1Name=dyn_jacketHigh_l
  bone2Name=spine_04
  swing1LimitDegrees=45
  swing1Motion=Free
  swing2LimitDegrees=45
  swing2Motion=Free
  twistLimitDegrees=45
  twistMotion=Free
 }
 {bone1Name=dyn_jacketString_01_l, bone2Name=dyn_jacketHigh_l, swing1LimitDegrees=15, swing2LimitDegrees=15, twistLimitDegrees=10}
 {bone1Name=dyn_jacketString_02_l, bone2Name=dyn_jacketString_01_l, swing1LimitDegrees=20, swing2LimitDegrees=20, twistLimitDegrees=10}
 {bone1Name=dyn_jacketString_03_l, bone2Name=dyn_jacketString_02_l, swing1LimitDegrees=20, swing2LimitDegrees=20, twistLimitDegrees=10}
 {bone1Name=dyn_jacketString_04_l, bone2Name=dyn_jacketString_03_l, swing1LimitDegrees=20, swing2LimitDegrees=20, twistLimitDegrees=10}
 {
  bone1Name=dyn_jacketLow_l
  bone2Name=spine_04
  swing1LimitDegrees=45
  swing1Motion=Free
  swing2LimitDegrees=45
  swing2Motion=Free
  twistLimitDegrees=45
  twistMotion=Free
 }
]
