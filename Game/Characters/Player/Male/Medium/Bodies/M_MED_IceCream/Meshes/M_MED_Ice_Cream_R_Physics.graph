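# Physics asset for the right-side jacket and drawstring of the medium male Ice Cream player body, giving its dynamic bones shapes for secondary motion.
# Capsules on spine_04 and spine_05 form the torso, with upper and lower jacket bodies each constrained to spine_04.
# A four-link capsule string chains off the upper jacket, with narrower swing and twist angles than that jacket's joint.
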PhysicsAsset /Game/Characters/Player/Male/Medium/Bodies/M_MED_IceCream/Meshes/M_MED_Ice_Cream_R_Physics
bodies {
 dyn_jacketHigh_r [{shapeType=Box, extentX=19.5526, extentY=21.2701, extentZ=3.3465, center=<0.8481 0.0933 -2.3025>, rotation=<rot 0 0 -1.1681>}]
 dyn_jacketLow_r [
  {shapeType=Capsule, length=17.5534, radius=12.4466, center=<9.6646 -8.3538 -7.5207>, rotation=<rot -12.4158 -1.7816 -3.9466>}
  {shapeType=Box, extentX=4.9779, extentY=31.2435, extentZ=31.5024, center=<10.1788 -0.9405 -0.1689>, rotation=<rot -5.9039 -87.6157 -79.4281>}
 ]
 dyn_jacketString_01_r [{shapeType=Capsule, length=5.612, radius=0.505, center=<3.0816 0.0025 -0.0354>, rotation=<rot 0 89.9496 -89.3791>}]
 dyn_jacketString_02_r [{shapeType=Capsule, length=6.1683, radius=0.7488, center=<3.6807 -0.0003 0>, rotation=<rot 0 89.9999 -89.9999>}]
 dyn_jacketString_03_r [{shapeType=Capsule, length=5.7507, radius=0.7847, center=<3.1391 0.0198 -0.0119>, rotation=<rot 90 0 0>}]
 dyn_jacketString_04_r [{shapeType=Capsule, length=1.7778, radius=1.0283, center=<1.1366 0.0163 -0.0307>, rotation=<rot 0 89.3624 -88.2267>}]
 spine_04 [{shapeType=Capsule, length=3.0215, radius=18.4005, rotation=<rot 90 7.125 -2.3139>}]
 spine_05 [{shapeType=Capsule, length=13.4212, radius=19.2068, center=<9.465 5.1599 1.0623>, rotation=<rot 90 0 -6.9319>}]
}
constraints [
 {bone1Name=spine_05, bone2Name=spine_04, swing1LimitDegrees=45, swing2LimitDegrees=45, twistLimitDegrees=45}
 {bone1Name=dyn_jacketHigh_r, bone2Name=spine_04, swing1LimitDegrees=45, swing2LimitDegrees=45, twistLimitDegrees=45}
 {bone1Name=dyn_jacketString_01_r, bone2Name=dyn_jacketHigh_r, swing1LimitDegrees=20, swing2LimitDegrees=20, twistLimitDegrees=10}
 {bone1Name=dyn_jacketString_02_r, bone2Name=dyn_jacketString_01_r, swing1LimitDegrees=25, swing2LimitDegrees=25, twistLimitDegrees=10}
 {bone1Name=dyn_jacketString_03_r, bone2Name=dyn_jacketString_02_r, swing1LimitDegrees=25, swing2LimitDegrees=25, twistLimitDegrees=10}
 {bone1Name=dyn_jacketString_04_r, bone2Name=dyn_jacketString_03_r, swing1LimitDegrees=20, swing2LimitDegrees=20, twistLimitDegrees=10}
 {bone1Name=dyn_jacketLow_r, bone2Name=spine_04, swing1LimitDegrees=35, swing2LimitDegrees=35, twistLimitDegrees=5}
]
