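# AnimBlueprint for the Ice Cream medium male player body, adding secondary motion to jacket, hood and cone bones.
# Parent-mesh pose feeds upper-arm Pose Drivers that correct the jacket for weapon poses.
# AnimDynamics, RigidBody simulations, aim-blended bone offsets and arm-driven Flap nodes then layer on top.
# A backbling check on customization load sets hood physics alpha; the standalone dyn_hood AnimDynamics node is unconnected.
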AnimBlueprint /Game/Characters/Player/Male/Medium/Bodies/M_MED_IceCream/Meshes/M_MED_Ice_Cream_AnimBP
target_skeleton=&/M_MED_Ice_Cream_Skeleton
graphs [
 AnimGraph {nodes [
   n1 = Get backblingAlpha-() -> (backblingAlpha-:double)
   n2 = Get HoodieExtForce() -> (HoodieExtForce:Vector)
   Anim Dynamics - Bone: dyn_hood(Alpha=@n1, ExternalForce=@n2) {
    angularDampingOverride=1
    angularSpringConstant=1000
    bAngularSpring=true
    bGravityOverrideInSimSpace=true
    bOverrideAngularDamping=true
    bUsePlanarLimit=false
    gravityScale=0.2
    simulationSpace=BoneRelative
    alphaBoolBlend {blendOption=Linear}
    alphaScaleBiasClamp {bClampResult=true, bInterpResult=true, clampMax=0.8, interpSpeedDecreasing=5, interpSpeedIncreasing=5}
    boundBone {boneName=dyn_hood}
    componentAppliedLinearAccClamp=<150 150 150>
    componentLinearAccScale=<0.6 0.6 0.6>
    componentLinearVelScale=<0.5 0.5 0.5>
    physicsBodyDefinitions [{
      collisionType=CustomSphere
      sphereCollisionRadius=3
      boundBone {boneName=dyn_hood}
      boxExtents=<5 20 20>
      constraintSetup {angularTargetAxis=AxisY, twistAxis=AxisY, angularLimitsMax=<20 20 30>, angularLimitsMin=<-20 -20 -30>, angularTarget=<0 1 0>}
      localJointOffset=<-6 -2 0>
     }]
    relativeSpaceBone {boneName=spine_03}
   }
   n3 = Get AimInAlpha() -> (AimInAlpha:float)
   n4 = Get HoodieExtForce()
   n5 = Anim Dynamics - Bone: dyn_jacketLow_r(ComponentPose=Local To Component(LocalPose=Use cached pose 'SavedPose11747'()), Alpha=@n3, ExternalForce=@n4) {
    angularDampingOverride=1
    angularSpringConstant=300
    bAngularSpring=true
    bGravityOverrideInSimSpace=true
    bOverrideAngularDamping=true
    bUsePlanarLimit=false
    gravityScale=0.15
    simulationSpace=RootRelative
    alphaBoolBlend {blendOption=Linear}
    alphaScaleBiasClamp {bClampResult=true, bInterpResult=true, clampMin=0.2, interpSpeedDecreasing=5, interpSpeedIncreasing=5}
    boundBone {boneName=dyn_jacketLow_r}
    componentAppliedLinearAccClamp=<500 500 500>
    componentLinearAccScale=<0.2 0.2 0.2>
    componentLinearVelScale=<0.2 0.2 0.2>
    physicsBodyDefinitions [{
      collisionType=CustomSphere
      sphereCollisionRadius=3
      boundBone {boneName=dyn_jacketLow_r}
      boxExtents=<15 3 18>
      constraintSetup {angularLimitsMax=<10 10 30>, angularLimitsMin=<-10 -10 -10>, angularTarget=<0 0 -1>}
      localJointOffset=<5 0 0>
     }]
    relativeSpaceBone {boneName=root}
   }
   n6 = Anim Dynamics - Bone: dyn_jacketLow_l(ComponentPose=@n5, Alpha=@n3, ExternalForce=@n4) {
    angularDampingOverride=1
    angularSpringConstant=300
    bAngularSpring=true
    bGravityOverrideInSimSpace=true
    bOverrideAngularDamping=true
    bUsePlanarLimit=false
    gravityScale=0.15
    simulationSpace=RootRelative
    alphaBoolBlend {blendOption=Linear}
    alphaScaleBiasClamp {bClampResult=true, bInterpResult=true, interpSpeedIncreasing=5}
    boundBone {boneName=dyn_jacketLow_l}
    componentAppliedLinearAccClamp=<400 400 400>
    componentLinearAccScale=<0.2 0.2 0.2>
    componentLinearVelScale=<0.2 0.2 0.2>
    physicsBodyDefinitions [{
      collisionType=CustomSphere
      sphereCollisionRadius=3
      boundBone {boneName=dyn_jacketLow_l}
      boxExtents=<15 3 15>
      constraintSetup {angularLimitsMax=<10 10 30>, angularLimitsMin=<-10 -10 -10>, angularTarget=<1 0 0>}
      localJointOffset=<-5 0 0>
     }]
    relativeSpaceBone {boneName=root}
   }
   n7 = Anim Dynamics - Bone: dyn_jacketBack(ComponentPose=@n6, Alpha=1, ExternalForce=@n4) {
    angularDampingOverride=1
    angularSpringConstant=500
    bAngularSpring=true
    bGravityOverrideInSimSpace=true
    bOverrideAngularDamping=true
    gravityScale=0.2
    simulationSpace=RootRelative
    alphaBoolBlend {blendOption=Linear}
    alphaScaleBiasClamp {bClampResult=true, bInterpResult=true, clampMin=0.2, interpSpeedDecreasing=5, interpSpeedIncreasing=5}
    boundBone {boneName=dyn_jacketBack}
    componentAppliedLinearAccClamp=<1000 1000 1000>
    componentLinearAccScale=<1 1 1>
    componentLinearVelScale=<0.5 0.5 0.5>
    physicsBodyDefinitions [{
      collisionType=CustomSphere
      sphereCollisionRadius=3
      boundBone {boneName=dyn_jacketBack}
      boxExtents=<10 3 10>
      constraintSetup {angularLimitsMax=<6 1 2>, angularLimitsMin=<-6 -1 -80>, angularTarget=<1 0 0>}
      localJointOffset=<5 0 0>
     }]
    planarLimits [{drivingBone {boneName=spine_03}, planeTransform {location=<0 -13.5 0>, rotation=<rot 0 10.0002 -89.9999>}}]
    relativeSpaceBone {boneName=root}
   }
   n8 = "Transform (Modify) Bone - Bone: dyn_ConeTop"(ComponentPose=@n7, Alpha=Get aimUp() -> (aimUp:float), Translation="0,7,0", Rotation="-10,0,0", Scale="1,1,1") {
    rotationMode=BMM_Additive
    scaleMode=BMM_Replace
    translationMode=BMM_Additive
    alphaBoolBlend {blendOption=Linear}
    boneToModify {boneName=dyn_ConeTop}
    rotation {pitch=-10}
    translation {y=7}
   }
   n9 = Get inVehicleAlpha() -> (inVehicleAlpha:float)
   n10 = RigidBody(ComponentPose=@n8, Alpha=@n9) {
    bClampLinearTranslationLimitToRefPose=true
    bOverrideWorldGravity=true
    overridePhysicsAsset=&/M_MED_Ice_Cream_R_Physics
    simulationSpace=BaseBoneSpace
    alphaBoolBlend {blendOption=Linear}
    alphaScaleBiasClamp {bClampResult=true, bInterpResult=true, clampMin=0.1, interpSpeedDecreasing=0.2, interpSpeedIncreasing=0.2}
    baseBoneRef {boneName=dyn_jacketHigh_r}
    componentAppliedLinearAccClamp=<200 200 200>
    componentLinearAccScale=<1 1 1>
    componentLinearVelScale=<0.4 0.4 0.4>
    overrideWorldGravity {z=-980}
   }
   n11 = RigidBody(ComponentPose=@n10, Alpha=@n9) {
    bClampLinearTranslationLimitToRefPose=true
    bOverrideWorldGravity=true
    overridePhysicsAsset=&/M_MED_Ice_Cream_L_Physics
    simulationSpace=BaseBoneSpace
    alphaBoolBlend {blendOption=Linear}
    alphaScaleBiasClamp {bClampResult=true, bInterpResult=true, clampMin=0.1, interpSpeedDecreasing=0.2, interpSpeedIncreasing=0.2}
    baseBoneRef {boneName=dyn_jacketHigh_l}
    componentAppliedLinearAccClamp=<200 200 200>
    componentLinearAccScale=<1 1 1>
    componentLinearVelScale=<0.4 0.4 0.4>
    overrideWorldGravity {z=-980}
   }
   n12 = RigidBody(ComponentPose=@n11, Alpha=@n1, ExternalForce=@n2) {
    bClampLinearTranslationLimitToRefPose=true
    bOverrideWorldGravity=true
    overridePhysicsAsset=&/M_MED_IceCream_Hoodie_Physics
    simulationSpace=BaseBoneSpace
    alphaBoolBlend {blendOption=Linear}
    alphaScaleBiasClamp {bClampResult=true, bInterpResult=true, clampMax=0.75, interpSpeedDecreasing=5, interpSpeedIncreasing=5}
    baseBoneRef {boneName=pelvis}
    componentAppliedLinearAccClamp=<50 50 50>
    componentLinearAccScale=<0.6 0.6 0.6>
    componentLinearVelScale=<2 2 2>
    overrideWorldGravity {z=-980}
   }
   n13 = Get aimIn() -> (aimIn:float)
   n14 = "Transform (Modify) Bone - Bone: dyn_jacketFix"(ComponentPose=@n12, Alpha=@n13, Translation="0,0,0", Rotation="0,60,0", Scale="1,1,1") {
    rotationMode=BMM_Additive
    rotationSpace=BCS_BoneSpace
    alphaBoolBlend {blendOption=Linear}
    boneToModify {boneName=dyn_jacketFix}
    rotation {yaw=60}
   }
   n15 = "Transform (Modify) Bone - Bone: dyn_jacketHigh_r"(ComponentPose=@n14, Alpha=@n13, Translation="-1,0,0", Rotation="-15,0,0", Scale="1,1,1") {
    rotationMode=BMM_Additive
    rotationSpace=BCS_BoneSpace
    scaleMode=BMM_Replace
    translationMode=BMM_Additive
    alphaBoolBlend {blendOption=Linear}
    boneToModify {boneName=dyn_jacketHigh_r}
    rotation {pitch=-15}
    translation {x=-1}
   }
   n16 = "Transform (Modify) Bone - Bone: dyn_jacketLowBase_l"(ComponentPose=@n15, Alpha=@n13, Translation="0,2,0", Rotation="0,6,0", Scale="1,1,1") {
    rotationMode=BMM_Additive
    rotationSpace=BCS_BoneSpace
    scaleMode=BMM_Replace
    translationMode=BMM_Additive
    translationSpace=BCS_BoneSpace
    alphaBoolBlend {blendOption=Linear}
    boneToModify {boneName=dyn_jacketLowBase_l}
    rotation {yaw=6}
    translation {y=2}
   }
   n17 = "Transform (Modify) Bone - Bone: dyn_jacketHigh_l"(ComponentPose=@n16, Alpha=Get aimUp(), Translation="0,0,-3", Rotation="0,-8,0", Scale="1,1,1") {
    rotationMode=BMM_Additive
    rotationSpace=BCS_BoneSpace
    scaleMode=BMM_Replace
    translationMode=BMM_Additive
    translationSpace=BCS_BoneSpace
    alphaBoolBlend {blendOption=Linear}
    alphaScaleBiasClamp {bInterpResult=true, interpSpeedDecreasing=6, interpSpeedIncreasing=6}
    boneToModify {boneName=dyn_jacketHigh_l}
    rotation {yaw=-8}
    translation {z=-3}
   }
   n18 = Flap Node - Bone: dyn_coneChestBase_r!(ComponentPose:Component Space Pose Link Structure=@n17, Alpha:float=1) -> (Pose:Component Space Pose Link Structure) {
    is=title FortAnimGraphNode_Flap
    bUseClamp=true
    targetClampMax=20
    targetClampMin=-3
    targetComponent=TranslationZ
    alphaBoolBlend {blendOption=Linear}
    sourceDrives [{
      bUseRange=false
      drivingCurve=null
      multiplier=1
      rangeMax=1
      rangeMin=-1
      remappedMax=1
      remappedMin=0
      sourceComponent=RotationZ
      useQuaternion=true
      sourceBone {boneName=upperarm_r}
     }]
    targetBone {boneName=dyn_coneChestBase_r}
   }
   n19 = Flap Node - Bone: dyn_coneChestBase_fr!(ComponentPose:Component Space Pose Link Structure=@n18, Alpha:float=1) -> (Pose:Component Space Pose Link Structure) {
    is=title FortAnimGraphNode_Flap
    bUseClamp=true
    targetClampMax=20
    targetClampMin=-20
    targetComponent=TranslationZ
    alphaBoolBlend {blendOption=Linear}
    sourceDrives [{
      bUseRange=false
      drivingCurve=null
      multiplier=5
      rangeMax=1
      rangeMin=-1
      remappedMax=1
      remappedMin=0
      sourceComponent=RotationZ
      useQuaternion=true
      sourceBone {boneName=upperarm_r}
     }]
    targetBone {boneName=dyn_coneChestBase_fr}
   }
   Output Pose(Result=Component To Local(ComponentPose=@n19)) {layerGroup=None}
   n20 = Pose Driver - Source: upperarm_r!(SourcePose:Pose Link Structure=Copy Pose From Mesh() {bCopyCurves=true, bUseAttachedParent=true}) -> (Pose:Pose Link Structure) {
    is=title AnimGraphNode_PoseDriver
    poseAsset=&/IceCream_AllWeapons_R_Posed_PoseAsset
    onlyDriveBones [{boneName=dyn_jacketHigh_r}, {boneName=dyn_jacketLowBase_r}]
    poseTargets [
     {
      bApplyCustomCurve=false
      bIsHidden=false
      distanceMethod=DefaultMethod
      drivenName=ref
      functionType=DefaultFunction
      targetScale=0.2276
      boneTransforms [{targetRotation=<rot -64.7854 -3.089 6.0959>, targetTranslation=<-15.5111 0 0>}]
      customCurve {defaultValue=3.4028234663852886e+38, postInfinityExtrap=RCCE_Constant, preInfinityExtrap=RCCE_Constant, keys []}
      targetRotation=<rot 0 0 0>
     }
     {
      bApplyCustomCurve=false
      bIsHidden=false
      distanceMethod=DefaultMethod
      drivenName=r_pis_for
      functionType=DefaultFunction
      targetScale=1
      boneTransforms [{targetRotation=<rot -3.3417 104.2977 -59.7618>, targetTranslation=<-15.5111 0 0>}]
      customCurve {defaultValue=3.4028234663852886e+38, postInfinityExtrap=RCCE_Constant, preInfinityExtrap=RCCE_Constant, keys []}
      targetRotation=<rot 0 0 0>
     }
     {
      bApplyCustomCurve=false
      bIsHidden=false
      distanceMethod=DefaultMethod
      drivenName=r_pis_dn
      functionType=DefaultFunction
      targetScale=0.148
      boneTransforms [{targetRotation=<rot -47.0027 108.2105 -46.0589>, targetTranslation=<-15.5111 0 0>}]
      customCurve {defaultValue=3.4028234663852886e+38, postInfinityExtrap=RCCE_Constant, preInfinityExtrap=RCCE_Constant, keys []}
      targetRotation=<rot 0 0 0>
     }
     {
      bApplyCustomCurve=false
      bIsHidden=false
      distanceMethod=DefaultMethod
      drivenName=r_pis_up
      functionType=DefaultFunction
      targetScale=1
      boneTransforms [{targetRotation=<rot 34.5408 94.743 -85.3418>, targetTranslation=<-15.5111 0 0>}]
      customCurve {defaultValue=3.4028234663852886e+38, postInfinityExtrap=RCCE_Constant, preInfinityExtrap=RCCE_Constant, keys []}
      targetRotation=<rot 0 0 0>
     }
     {
      bApplyCustomCurve=false
      bIsHidden=false
      distanceMethod=DefaultMethod
      drivenName=r_rif_for
      functionType=DefaultFunction
      targetScale=0.2503
      boneTransforms [{targetRotation=<rot -55.8149 101.5018 -61.4575>, targetTranslation=<-15.5111 0 0>}]
      customCurve {defaultValue=3.4028234663852886e+38, postInfinityExtrap=RCCE_Constant, preInfinityExtrap=RCCE_Constant, keys []}
      targetRotation=<rot 0 0 0>
     }
     {
      bApplyCustomCurve=false
      bIsHidden=false
      distanceMethod=DefaultMethod
      drivenName=r_rif_dn
      functionType=DefaultFunction
      targetScale=0.4824
      boneTransforms [{targetRotation=<rot -72.7903 121.1539 -76.4961>, targetTranslation=<-15.5111 0 0>}]
      customCurve {defaultValue=3.4028234663852886e+38, postInfinityExtrap=RCCE_Constant, preInfinityExtrap=RCCE_Constant, keys []}
      targetRotation=<rot 0 0 0>
     }
     {
      bApplyCustomCurve=false
      bIsHidden=false
      distanceMethod=DefaultMethod
      drivenName=r_rif_up
      functionType=DefaultFunction
      targetScale=0.148
      boneTransforms [{targetRotation=<rot -42.5204 113.317 -74.3445>, targetTranslation=<-15.5111 0 0>}]
      customCurve {defaultValue=3.4028234663852886e+38, postInfinityExtrap=RCCE_Constant, preInfinityExtrap=RCCE_Constant, keys []}
      targetRotation=<rot 0 0 0>
     }
     {
      bApplyCustomCurve=false
      bIsHidden=false
      distanceMethod=DefaultMethod
      drivenName=r_mg_for
      functionType=DefaultFunction
      targetScale=0.4144
      boneTransforms [{targetRotation=<rot -60.8709 -37.8721 81.8924>, targetTranslation=<-15.5111 0 0>}]
      customCurve {defaultValue=3.4028234663852886e+38, postInfinityExtrap=RCCE_Constant, preInfinityExtrap=RCCE_Constant, keys []}
      targetRotation=<rot 0 0 0>
     }
     {
      bApplyCustomCurve=false
      bIsHidden=false
      distanceMethod=DefaultMethod
      drivenName=r_mg_dn
      functionType=DefaultFunction
      targetScale=0.2276
      boneTransforms [{targetRotation=<rot -64.7805 17.805 33.6797>, targetTranslation=<-15.5111 0 0>}]
      customCurve {defaultValue=3.4028234663852886e+38, postInfinityExtrap=RCCE_Constant, preInfinityExtrap=RCCE_Constant, keys []}
      targetRotation=<rot 0 0 0>
     }
     {
      bApplyCustomCurve=false
      bIsHidden=false
      distanceMethod=DefaultMethod
      drivenName=r_mg_up
      functionType=DefaultFunction
      targetScale=0.3942
      boneTransforms [{targetRotation=<rot -80.0599 2.4533 31.4588>, targetTranslation=<-15.5111 0 0>}]
      customCurve {defaultValue=3.4028234663852886e+38, postInfinityExtrap=RCCE_Constant, preInfinityExtrap=RCCE_Constant, keys []}
      targetRotation=<rot 0 0 0>
     }
    ]
    rBFParams {radius=40}
    sourceBones [{boneName=upperarm_r}]
   }
   n21 = Pose Driver - Source: upperarm_l!(SourcePose:Pose Link Structure=@n20) -> (Pose:Pose Link Structure) {
    is=title AnimGraphNode_PoseDriver
    poseAsset=&IceCream_AllWeapons_R_Posed_PoseAsset
    onlyDriveBones [{boneName=dyn_jacketHigh_l}, {boneName=dyn_jacketLowBase_l}]
    poseTargets [
     {
      bApplyCustomCurve=false
      bIsHidden=false
      distanceMethod=DefaultMethod
      drivenName=ref
      functionType=DefaultFunction
      targetScale=0.5678
      boneTransforms [{targetRotation=<rot -64.7854 -3.089 6.0959>, targetTranslation=<15.5111 0 0>}]
      customCurve {defaultValue=3.4028234663852886e+38, postInfinityExtrap=RCCE_Constant, preInfinityExtrap=RCCE_Constant, keys []}
      targetRotation=<rot 0 0 0>
     }
     {
      bApplyCustomCurve=false
      bIsHidden=false
      distanceMethod=DefaultMethod
      drivenName=r_pis_for
      functionType=DefaultFunction
      targetScale=0.5598
      boneTransforms [{targetRotation=<rot -30.4689 70.4139 -76.2345>, targetTranslation=<15.5111 0 0>}]
      customCurve {defaultValue=3.4028234663852886e+38, postInfinityExtrap=RCCE_Constant, preInfinityExtrap=RCCE_Constant, keys []}
      targetRotation=<rot 0 0 0>
     }
     {
      bApplyCustomCurve=false
      bIsHidden=false
      distanceMethod=DefaultMethod
      drivenName=r_pis_dn
      functionType=DefaultFunction
      targetScale=0.6471
      boneTransforms [{targetRotation=<rot -75.7311 101.1543 -79.9244>, targetTranslation=<15.5111 0 0>}]
      customCurve {defaultValue=3.4028234663852886e+38, postInfinityExtrap=RCCE_Constant, preInfinityExtrap=RCCE_Constant, keys []}
      targetRotation=<rot 0 0 0>
     }
     {
      bApplyCustomCurve=false
      bIsHidden=false
      distanceMethod=DefaultMethod
      drivenName=r_pis_up
      functionType=DefaultFunction
      targetScale=1
      boneTransforms [{targetRotation=<rot 11.1408 56.259 -101.0537>, targetTranslation=<15.5111 0 0>}]
      customCurve {defaultValue=3.4028234663852886e+38, postInfinityExtrap=RCCE_Constant, preInfinityExtrap=RCCE_Constant, keys []}
      targetRotation=<rot 0 0 0>
     }
     {
      bApplyCustomCurve=false
      bIsHidden=false
      distanceMethod=DefaultMethod
      drivenName=r_rif_for
      functionType=DefaultFunction
      targetScale=0.4541
      boneTransforms [{targetRotation=<rot -40.1451 44.3494 -69.5741>, targetTranslation=<15.5111 0 0>}]
      customCurve {defaultValue=3.4028234663852886e+38, postInfinityExtrap=RCCE_Constant, preInfinityExtrap=RCCE_Constant, keys []}
      targetRotation=<rot 0 0 0>
     }
     {
      bApplyCustomCurve=false
      bIsHidden=false
      distanceMethod=DefaultMethod
      drivenName=r_rif_dn
      functionType=DefaultFunction
      targetScale=0.5386
      boneTransforms [{targetRotation=<rot -64.4481 42.2487 -24.2281>, targetTranslation=<15.5111 0 0>}]
      customCurve {defaultValue=3.4028234663852886e+38, postInfinityExtrap=RCCE_Constant, preInfinityExtrap=RCCE_Constant, keys []}
      targetRotation=<rot 0 0 0>
     }
     {
      bApplyCustomCurve=false
      bIsHidden=false
      distanceMethod=DefaultMethod
      drivenName=r_rif_up
      functionType=DefaultFunction
      targetScale=1
      boneTransforms [{targetRotation=<rot -6.1697 27.4368 -79.4979>, targetTranslation=<15.5111 0 0>}]
      customCurve {defaultValue=3.4028234663852886e+38, postInfinityExtrap=RCCE_Constant, preInfinityExtrap=RCCE_Constant, keys []}
      targetRotation=<rot 0 0 0>
     }
     {
      bApplyCustomCurve=false
      bIsHidden=false
      distanceMethod=DefaultMethod
      drivenName=r_mg_for
      functionType=DefaultFunction
      targetScale=0.2036
      boneTransforms [{targetRotation=<rot -49.1947 71.8213 -33.9508>, targetTranslation=<15.5111 0 0>}]
      customCurve {defaultValue=3.4028234663852886e+38, postInfinityExtrap=RCCE_Constant, preInfinityExtrap=RCCE_Constant, keys []}
      targetRotation=<rot 0 0 0>
     }
     {
      bApplyCustomCurve=false
      bIsHidden=false
      distanceMethod=DefaultMethod
      drivenName=r_mg_dn
      functionType=DefaultFunction
      targetScale=0.2036
      boneTransforms [{targetRotation=<rot -49.5185 61.3502 -47.2735>, targetTranslation=<15.5111 0 0>}]
      customCurve {defaultValue=3.4028234663852886e+38, postInfinityExtrap=RCCE_Constant, preInfinityExtrap=RCCE_Constant, keys []}
      targetRotation=<rot 0 0 0>
     }
     {
      bApplyCustomCurve=false
      bIsHidden=false
      distanceMethod=DefaultMethod
      drivenName=r_mg_up
      functionType=DefaultFunction
      targetScale=0.7204
      boneTransforms [{targetRotation=<rot -13.591 89.1385 -66.9031>, targetTranslation=<15.5111 0 0>}]
      customCurve {defaultValue=3.4028234663852886e+38, postInfinityExtrap=RCCE_Constant, preInfinityExtrap=RCCE_Constant, keys []}
      targetRotation=<rot 0 0 0>
     }
    ]
    rBFParams {radius=40}
    sourceBones [{boneName=upperarm_l}]
   }
   Save cached pose 'SavedPose11747'(Pose=@n21) {cacheName=SavedPose11747}
   Sequence Player 'IceCream_AllWeapons_Posed'() {sequence=&/IceCream_AllWeapons_Posed}
   Sequence Player 'IceCream_AllWeapons'() {sequence=&/IceCream_AllWeapons}
  ]}
 EventGraph {nodes [
   Event OnCharacterCustomizationLoaded()
   n1 = CheckForBackbling()
   > n2 = Set isWearingBackbling(isWearingBackbling=OR Boolean(A=@n1, B=@n1.bIsBackblingLong))
   > Set backblingAlpha-(backblingAlpha-:double=float - float(B="To Float (Boolean)"(InBool=@n2)) {is=title K2Node_CallFunction}) -> (Output_Get:double)
   n3 = GetPartSkeletalMeshComponent()
   "Not Equal (Object)"!(
    A:Object Reference=FindChildCommponentFromAttachSocket!(
     self:Custom Character Part Anim Instance Object Reference
     Parent:Skeletal Mesh Component Object Reference=@n3
     bIncludeAllDescendants:bool=false
     AttachSocket:Name=Cape
    ) -> (ReturnValue:Skeletal Mesh Component Object Reference)
    B:Object Reference
   ) -> (ReturnValue:bool)
   "Not Equal (Object)"(A=FindChildCommponentFromAttachSocket(Parent=@n3, bIncludeAllDescendants=false, AttachSocket=backpack_br))
   Reroute Node(InputPin:Wildcard) -> (OutputPin:Wildcard)
  ]}
]
variables [
 backblingAlpha- {type=number}
 isWearingBackbling {type=boolean}
]
